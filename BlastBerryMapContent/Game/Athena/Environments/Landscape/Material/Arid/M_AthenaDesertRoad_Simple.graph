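# Desert road material for spline meshes in the Athena arid landscape, with Nanite support and an asphalt physical material.
# World-projected sand is color-tinted, with a biome-colormap variant for low quality and mobile, then passed through a desert road function that yields a road mask.
# Optional topo adjustment recolors the base.
# The material writes to the terrain Runtime Virtual Texture and can blend with it through static switches.
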Material /BlastBerryMapContent/Game/Athena/Environments/Landscape/Material/Arid/M_AthenaDesertRoad_Simple
graph {
 properties {
  bTangentSpaceNormal=false
  bUseMaterialAttributes=true
  bUsedWithNanite=true
  bUsedWithSplineMeshes=true
  physMaterial=/Game/FortressPhysicalMaterials/Asphalt
  pixelDepthOffsetMode=PDOM_Legacy
  refractionMethod=RM_IndexOfRefraction
  usedShadingModels=Default Lit
 }
 nodes [
  Constant3Vector() {constant=<rgba 0 1 0 0>}
  Constant()
  n1 = WorldPosition() {worldPositionShaderOffset=WPT_ExcludeAllShaderOffsets}
  n2 = Multiply(A=Abs(None=CollectionParameter() {collection=/BlastBerryMapContent/Game/Athena/Environments/Landscape/MPC/MPC_Landscape, parameterName=Arid_Sand_TexSize})) {constB=-1}
  n3 = MaterialFunctionCall!(UseRoughness=StaticBool(), UVs=ComponentMask(None=Divide(A=@n1.XYZ, B=@n2)) {g=true, r=true}) {materialFunction=/BlastBerryMapContent/Game/Athena/Environments/Landscape/MaterialFunctions/Arid/MF_Athena_Arid_Sand}
  n4 = ComponentMask(None=WorldPosition().XYZ) {g=true, r=true}
  n5 = ScalarParameter() {defaultValue=306000, parameterName=LandscapeSize, sortPriority=-153000}
  n6 = ScalarParameter() {defaultValue=-153000, parameterName=LandscapeMin, sortPriority=-153000}
  n7 = TextureSampleParameter2D(UVs=Subtract(A=Divide(A=@n4, B=@n5), B=Divide(A=@n6, B=@n5))) {
   parameterName=BiomeColormap
   samplerType=SAMPLERTYPE_LinearColor
   texture=/BlastBerryMapContent/Game/Athena/Environments/Landscape/Textures/T_Athena_Terrain_CombinedColors_01
  }
  n8 = CollectionParameter() {collection=&MPC_Landscape, parameterName=ColormapDivisor}
  n9 = BreakMaterialAttributes!(Attr=MaterialFunctionCall!(In=@n3.Simple_Material, Color=Multiply(A=@n7.RGB, B=@n8)) {materialFunction=/Game/Athena/Environments/Landscape/MaterialFunctions/MF_Athena_Material_BaseMultiply}.Result)
  n10 = MakeMaterialAttributes!(
   BaseColor=@n9.BaseColor
   Metallic
   Specular=@n9.BaseColor
   Roughness=Constant() {r=1}
   Anisotropy
   EmissiveColor
   Opacity
   OpacityMask
   Normal=@n9.BaseColor
   Tangent
   WorldPositionOffset=@n9.BaseColor
   SubsurfaceColor
   ClearCoat
   ClearCoatRoughness
   AmbientOcclusion
   Refraction
   CustomizedUVs_0
   CustomizedUVs_1
   CustomizedUVs_2
   CustomizedUVs_3
   CustomizedUVs_4
   CustomizedUVs_5
   CustomizedUVs_6
   CustomizedUVs_7
   PixelDepthOffset
   ShadingModel
   Displacement
  )
  n11 = QualitySwitch(
   Default=MaterialFunctionCall(
    In=@n3.Simple_Material
    Color=MaterialFunctionCall() {materialFunction=/BlastBerryMapContent/Game/Athena/Environments/Landscape/MaterialFunctions/MF_Athena_ZSandColor}.Result
   ) {materialFunction=&MF_Athena_Material_BaseMultiply}.Result
   Low=@n10
  )
  n12 = FeatureLevelSwitch(Default=@n11, ES2=@n10, ES3_1=@n10, SM4_REMOVED=@n11, SM5=@n11, SM6=@n11)
  n13 = MaterialFunctionCall!(In=@n12) {materialFunction=/BlastBerryMapContent/Game/Athena/Environments/Landscape/MaterialFunctions/Arid/MF_Athena_DesertRoad}
  n14 = StaticSwitchParameter(True=@n13.Result, False=@n12) {defaultValue=true, parameterName=UseDesertRoad}
  n15 = WorldPosition()
  n16 = ScalarParameter() {defaultValue=306200, group=TerrainColorMatch, parameterName=TerrainSize}
  n17 = Divide(A=@n16)
  n18 = Divide(A=Subtract(A=ComponentMask(None=@n15.XYZ) {g=true, r=true}, B=AppendVector(A=@n17, B=@n17)), B=@n16)
  n19 = MaterialFunctionCall!(
   Mask 2=TextureObject() {samplerType=SAMPLERTYPE_LinearColor, texture=/BlastBerryMapContent/Game/Athena/Environments/Textures/T_Athena_Terrain_Topo_Mask}
   UV Coordinates=VertexInterpolator(VS=@n18).PS
   RockMask
   GrassMask=Constant() {r=0.5}
   Arid=Constant() {r=1}
   BaseTexture=MaterialFunctionCall!(Material=@n14) {materialFunction=/Engine/Functions/MaterialLayerFunctions/MatLayerBlend_BreakBaseColor}.BaseColor
  ) {materialFunction=/BlastBerryMapContent/Game/Athena/Environments/Materials/MaterialFunctions/MF_TerrainTopoAdjustment}
  n20 = Multiply(
   A=Desaturation(
    None=@n19.Result
    Fraction=CollectionParameter() {collection=&MPC_Landscape, parameterName=Arid_EdgeDesaturation}
   )
   B=ComponentMask(None=CollectionParameter() {collection=&MPC_Landscape, parameterName=Arid_SandSediment_Tint}) {b=true, g=true, r=true}
  )
  n21 = StaticSwitchParameter(
   True=MaterialFunctionCall!(Mat_In=@n14, BaseColor=LinearInterpolate(A=@n19.Result, B=@n20, Alpha=@n19.Result)) {materialFunction=/Game/Athena/Environments/Landscape/MaterialFunctions/MF_Athena_ReplaceBaseColor}.Result
   False=@n14
  ) {defaultValue=true, parameterName=UseTopo}
  n22 = NamedRerouteUsage()
  n23 = NamedRerouteUsage()
  n24 = OneMinus(None=@n23)
  n25 = GetMaterialAttributes(None=@n21) {attributeGetTypes [
    #1
    #2
    #3
    #4
   ]}
  n26 = StaticSwitchParameter(
   True=MaterialFunctionCall(NoRVT=@n21, RVT=BlendMaterialAttributes(A=@n21, B=@n22, Alpha=@n24)) {materialFunction=&FN_RVT_Switch_Material}.Result
   False=@n25.MaterialAttributes
  ) {defaultValue=true, group=D_RVT, parameterName=Use_RVT Blending, sortPriority=-1}
  MaterialFunctionCall!(Material=@n26) {materialFunction=/Game/Athena/Apollo/Maps/Landscape/VT/MaterialFunctions/TransformTangentToWorld}
  MaterialFunctionCall!(In=@n26) {materialFunction=/Game/Athena/Asteria/Environments/Landscape/Materials/MaterialFunctions/MF_Terrain_LumenModification}
  NamedRerouteDeclaration(None=@n13.RoadMask) {name=RoadMask, nodeColor=<rgba 0 1 0.3438 1>}
  n27 = RuntimeVirtualTextureSample!(UVs, World Position) {
   materialType=BaseColor_Normal_Specular_YCoCg
   textureAddressMode=RVTTA_Wrap
   virtualTexture=/BlastBerryMapContent/Game/Athena/Environments/Materials/RVT/Athena_RuntimeVirtualTexture
  }
  n28 = SetMaterialAttributes!(MaterialAttributes, Base Color=@n27.BaseColor, Specular=@n27.BaseColor, Roughness=@n27.BaseColor, Normal=@n27.BaseColor) {attributeSetTypes [
    #1
    #2
    #3
    #4
   ]}
  NamedRerouteDeclaration(None=@n28) {name=RVT-In, nodeColor=<rgba 0.5469 0 1 1>}
  n29 = ComponentMask(None=WorldPosition().XYZ) {b=true}
  RuntimeVirtualTextureOutput!(
   BaseColor=@n25.Base Color
   Specular=@n25.Base Color
   Roughness=@n25.Base Color
   Normal=@n25.Base Color
   WorldHeight=LinearInterpolate(A=Subtract(A=@n29) {constB=10000}, B=@n29, Alpha=@n23)
   Opacity=@n23
   Mask
   Displacement
   Mask4
  )
  TextureSample(UVs=@n18) {
   samplerSource=SSM_Wrap_WorldGroupSettings
   samplerType=SAMPLERTYPE_LinearColor
   texture=/BlastBerryMapContent/Game/Athena/Environments/Landscape/Textures/T_Athena_Terrain_Biome_M
  }
 ]
}
