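# Master material for Arid biome landscape-matching static and spline meshes, such as cliffs, roads and river gravel.
# Vertex colors drive height-lerped sand, grass, mud, road and gravel layers, with static switches selecting variants like cliff rock or desert road.
# World-XY layer normals are reoriented to the mesh surface, and optional topo masks adjust color to the terrain.
# Mobile and low quality use simpler shading, before RVT blending and Lumen modification functions.
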Material /BlastBerryMapContent/Game/Athena/Environments/Landscape/Material/Arid/M_Athena_AridSM_Master
graph {
 properties {
  bUseMaterialAttributes=true
  bUsedWithNanite=true
  bUsedWithSplineMeshes=true
  physMaterial=/Game/FortressPhysicalMaterials/Sand
  pixelDepthOffsetMode=PDOM_Legacy
  refractionMethod=RM_IndexOfRefraction
  usedShadingModels=Default Lit
 }
 nodes [
  Constant3Vector() {constant=<rgba 0 0 1 0>}
  Constant() {r=1}
  LandscapeLayerCoords() {mappingScale=1525}
  LandscapeLayerCoords() {mappingScale=1525}
  LandscapeLayerSample() {parameterName=Arid}
  n1 = ScalarParameter() {defaultValue=306200, group=TerrainColorMatch, parameterName=TerrainSize}
  n2 = Divide(A=@n1)
  n3 = Subtract(A=ComponentMask(None=WorldPosition().XYZ) {g=true, r=true}, B=AppendVector(A=@n2, B=@n2))
  n4 = TextureSample(UVs=VertexInterpolator(VS=Divide(A=@n3, B=@n1)).PS) {
   samplerSource=SSM_Wrap_WorldGroupSettings
   samplerType=SAMPLERTYPE_LinearColor
   texture=/BlastBerryMapContent/Game/Athena/Environments/Landscape/Textures/T_Athena_Terrain_Biome_M
  }
  n5 = MaterialFunctionCall!(
   Mask 2=TextureObject() {samplerType=SAMPLERTYPE_LinearColor, texture=/BlastBerryMapContent/Game/Athena/Environments/Textures/T_Athena_Terrain_Topo_Mask}
   UV Coordinates=@n4.B
   RockMask
   GrassMask=Constant() {r=1}
   Arid=@n4.B
   BaseTexture=MaterialFunctionCall!(Material) {materialFunction=/Engine/Functions/MaterialLayerFunctions/MatLayerBlend_BreakBaseColor}.BaseColor
  ) {materialFunction=/BlastBerryMapContent/Game/Athena/Environments/Materials/MaterialFunctions/MF_TerrainTopoAdjustment}
  n6 = Desaturation(
   None=@n5.Result
   Fraction=CollectionParameter() {collection=/BlastBerryMapContent/Game/Athena/Environments/Landscape/MPC/MPC_Landscape, parameterName=Arid_EdgeDesaturation}
  )
  n7 = ComponentMask(None=CollectionParameter() {collection=&MPC_Landscape, parameterName=Arid_SandSediment_Tint}) {b=true, g=true, r=true}
  MaterialFunctionCall!(Mat_In, BaseColor=LinearInterpolate(A=@n5.Result, B=Multiply(A=@n6, B=@n7), Alpha=@n5.Result)) {materialFunction=/Game/Athena/Environments/Landscape/MaterialFunctions/MF_Athena_ReplaceBaseColor}
  n8 = MaterialFunctionCall!(ProjContrast) {materialFunction=/BlastBerryMapContent/Game/Athena/Environments/Landscape/MaterialFunctions/Arid/MF_Athena_Arid_Rock}
  n9 = MaterialFunctionCall() {materialFunction=/BlastBerryMapContent/Game/Athena/Environments/Landscape/MaterialFunctions/MF_Athena_SedimentGradient}
  n10 = MaterialFunctionCall!(
   Material=MaterialFunctionCall!(
    In=MaterialFunctionCall(In=@n8.Mat, Color=@n9.Result) {materialFunction=/Game/Athena/Environments/Landscape/MaterialFunctions/MF_Athena_Material_BaseMultiply}.Result
    Color=ScalarParameter() {defaultValue=0.8, parameterName=Arid_CliffMesh_Brightness}
   ) {materialFunction=&MF_Athena_Material_BaseMultiply}.Result
   NewRoughness=Add(A=GetMaterialAttributes(None=@n8.Mat) {attributeGetTypes [#1]}.Roughness) {constB=0.35, desc=use to simulate the roughness the landscape has}
   Mask
  ) {materialFunction=/Engine/Functions/MaterialLayerFunctions/MatLayerBlend_RoughnessOverride}
  n11 = MaterialFunctionCall(
   In=MaterialFunctionCall() {materialFunction=/BlastBerryMapContent/Game/Athena/Environments/Landscape/MaterialFunctions/Standard/MF_Athena_Grass_01}.Result
   Color=CollectionParameter() {collection=&MPC_Landscape, parameterName=Arid_Grass_Tint}
  ) {materialFunction=&MF_Athena_Material_BaseMultiply}
  n12 = WorldPosition() {worldPositionShaderOffset=WPT_ExcludeAllShaderOffsets}
  n13 = Multiply(A=Abs(None=CollectionParameter() {collection=&MPC_Landscape, parameterName=Arid_Sand_TexSize})) {constB=-1}
  n14 = MaterialFunctionCall!(UseRoughness=StaticBool(), UVs=ComponentMask(None=Divide(A=@n12.XYZ, B=@n13)) {g=true, r=true}) {materialFunction=/BlastBerryMapContent/Game/Athena/Environments/Landscape/MaterialFunctions/Arid/MF_Athena_Arid_Sand}
  n15 = MaterialFunctionCall() {materialFunction=/BlastBerryMapContent/Game/Athena/Environments/Landscape/MaterialFunctions/MF_Athena_ZSandColor}
  n16 = VectorParameter() {parameterName=sandtint, defaultValue=<rgba 1 1 1 1>}
  n17 = ComponentMask(None=WorldPosition().XYZ) {g=true, r=true}
  n18 = ScalarParameter() {defaultValue=306000, parameterName=LandscapeSize, sortPriority=-153000}
  n19 = ScalarParameter() {defaultValue=-153000, parameterName=LandscapeMin, sortPriority=-153000}
  n20 = TextureSampleParameter2D(UVs=Subtract(A=Divide(A=@n17, B=@n18), B=Divide(A=@n19, B=@n18))) {
   parameterName=BiomeColormap
   samplerType=SAMPLERTYPE_LinearColor
   texture=/BlastBerryMapContent/Game/Athena/Environments/Landscape/Textures/T_Athena_Terrain_CombinedColors_01
  }
  n21 = MaterialFunctionCall(
   In=@n14.Simple_Material
   Color=Multiply(
    A=@n20.RGB
    B=CollectionParameter() {collection=&MPC_Landscape, parameterName=ColormapDivisor}
   )
  ) {materialFunction=&MF_Athena_Material_BaseMultiply}
  n22 = QualitySwitch(
   Default=MaterialFunctionCall(In=@n14.Simple_Material, Color=Multiply(A=@n15.Result, B=@n16.RGB)) {materialFunction=&MF_Athena_Material_BaseMultiply}.Result
   Low=@n21.Result
  )
  n23 = ShadingPathSwitch(Default=@n22, Deferred=@n22, Mobile=@n21.Result) {desc=Migrated from FeatureLevelSwitch}
  n24 = StaticSwitchParameter(True=@n11.Result, False=@n23) {parameterName=UseGrassAsBase}
  n25 = Abs(None=ScalarParameter() {defaultValue=1024, group=MasterControl, parameterName=Texturesize_Z})
  n26 = ComponentMask(None=Divide(A=WorldPosition() {worldPositionShaderOffset=WPT_ExcludeAllShaderOffsets}.XYZ, B=Multiply(A=@n25) {constB=-1})) {g=true, r=true}
  n27 = MaterialFunctionCall!(In=@n26) {materialFunction=/BlastBerryMapContent/Game/Athena/Environments/Landscape/MaterialFunctions/Standard/MF_Athena_Mud_01}
  n28 = Multiply(
   A=@n27.BaseColor_Plain
   B=ComponentMask(None=CollectionParameter() {collection=&MPC_Landscape, parameterName=Arid_Mud_Tint}) {b=true, g=true, r=true}
  )
  n29 = MaterialFunctionCall!(
   Material=@n27.Result
   NewBaseColor=@n28
   Mask=MaterialFunctionCall() {materialFunction=/BlastBerryMapContent/Game/Environments/Materials/Grass/MF_Athena_BiomeMask}.Swamp
  ) {materialFunction=/Engine/Functions/MaterialLayerFunctions/MatLayerBlend_OverrideBaseColor}
  n30 = VertexColor()
  n31 = StaticSwitchParameter(True=OneMinus(None=@n30.A), False=@n30.A) {parameterName=IsDesertRoad}
  n32 = MaterialFunctionCall!(Mat attributes=@n11.Result) {materialFunction=/Engine/Functions/Engine_MaterialFunctions02/MatLayerBlend_BreakOpacity}
  n33 = StaticSwitchParameter(
   True=OneMinus(None=@n32.Result)
   False=MaterialFunctionCall(Mat attributes=@n27.Result) {materialFunction=&MatLayerBlend_BreakOpacity}.Result
  ) {parameterName=UseGrassAsBase}
  n34 = MaterialFunctionCall!(
   Base Material=@n24
   Top Material=@n29.Blended Material
   Alpha=MaterialFunctionCall(Transition Phase=Clamp(None=@n31), Height Texture=@n33, Contrast=ScalarParameter() {defaultValue=1, parameterName=Mud_Blend_Contrast}) {desc=Mud, materialFunction=&HeightLerp}.Alpha
  ) {materialFunction=/Engine/Functions/MaterialLayerFunctions/MatLayerBlend_Standard}
  n35 = VertexColor()
  n36 = MaterialFunctionCall!(UseSimple=StaticBool(), In=@n26) {materialFunction=/BlastBerryMapContent/Game/Athena/Environments/Landscape/MaterialFunctions/Standard/MF_Athena_Road_01}
  n37 = MaterialFunctionCall(
   Transition Phase=Power(Base=@n35.R, Exp=ScalarParameter() {defaultValue=2, parameterName=RoadMud_Trans})
   Height Texture=MaterialFunctionCall(Mat attributes=@n36.Result) {materialFunction=&MatLayerBlend_BreakOpacity}.Result
   Contrast=ScalarParameter() {defaultValue=2, parameterName=RoadMud_Contrast}
  ) {materialFunction=&HeightLerp}
  n38 = OneMinus(None=@n37.Alpha)
  n39 = MaterialFunctionCall!(Material=@n29.Blended Material, Tint=@n38, TintMask=Constant() {r=1}) {materialFunction=/Engine/Functions/MaterialLayerFunctions/MatLayerBlend_Tint}
  n40 = Multiply(A=@n35.R, B=ScalarParameter() {defaultValue=1.5, parameterName=RoadMud_Trans_1})
  n41 = ScalarParameter() {defaultValue=1, parameterName=Road_Blend_Contrast}
  n42 = StaticSwitchParameter(
   True=@n24
   False=MaterialFunctionCall(
    Base Material=StaticSwitchParameter(True=@n34.Blended Material, False=@n24) {defaultValue=true, parameterName=UseMud}
    Top Material=MaterialFunctionCall(Base Material=@n39.Blended Material, Top Material=@n36.Result, Alpha=@n37.Alpha) {materialFunction=&MatLayerBlend_Standard}.Blended Material
    Alpha=MaterialFunctionCall(Transition Phase=@n40, Height Texture=@n33, Contrast=@n41) {desc=Road, materialFunction=&HeightLerp}.Alpha
   ) {materialFunction=&MatLayerBlend_Standard}.Blended Material
  ) {defaultValue=true, parameterName=Use_Cliffs}
  n43 = MaterialFunctionCall!(In=@n42) {materialFunction=/BlastBerryMapContent/Game/Athena/Environments/Landscape/MaterialFunctions/Arid/MF_Athena_DesertRoad}
  n44 = MaterialFunctionCall!(In=@n26) {materialFunction=/BlastBerryMapContent/Game/Athena/Environments/Landscape/MaterialFunctions/Standard/MF_Athena_Gravel_01}
  n45 = CollectionParameter() {collection=&MPC_Landscape, parameterName=Arid_Gravel_Tint}
  n46 = MaterialFunctionCall(Mat_In=@n44.Result, BaseColor=Multiply(A=@n44.Plain_BaseColor, B=ComponentMask(None=@n45) {b=true, g=true, r=true})) {materialFunction=&MF_Athena_ReplaceBaseColor}
  n47 = MaterialFunctionCall(In=@n46.Result, Color=VectorParameter() {parameterName=Gravel_Spline_Match, defaultValue=<rgba 1 0.7689 0.6574 0>}.RGB) {materialFunction=&MF_Athena_Material_BaseMultiply}
  n48 = VertexColor()
  n49 = ScalarParameter() {defaultValue=2, parameterName=Gravel_Blend_Contrast}
  n50 = OneMinus(None=@n32.Result)
  n51 = MaterialFunctionCall(
   Base Material=MaterialFunctionCall(
    Base Material=StaticSwitchParameter(True=@n43.Result, False=@n42) {parameterName=IsDesertRoad}
    Top Material=StaticSwitchParameter(True=@n47.Result, False=@n46.Result) {desc=For river spline, parameterName=MatchGravel}
    Alpha=MaterialFunctionCall(Transition Phase=@n48.B, Height Texture=@n44.Height, Contrast=@n49) {desc=Gravel, materialFunction=&HeightLerp}.Alpha
   ) {materialFunction=&MatLayerBlend_Standard}.Blended Material
   Top Material=StaticSwitchParameter(True=@n23, False=StaticSwitchParameter(True=@n23, False=@n11.Result) {desc=Flipped, parameterName=UseGrassAsBase}) {parameterName=NoGrass}
   Alpha=MaterialFunctionCall(
    Transition Phase=VertexColor().G
    Height Texture=StaticSwitchParameter(True=@n50, False=@n32.Result) {parameterName=UseGrassAsBase}
    Contrast=ScalarParameter() {parameterName=Grass_Blend_Contrast}
   ) {materialFunction=&HeightLerp}.Alpha
  ) {materialFunction=&MatLayerBlend_Standard}
  n52 = MaterialFunctionCall(
   Transition Phase=OneMinus(None=VertexColor().A)
   Height Texture=Constant() {r=0.5}
   Contrast=ScalarParameter() {defaultValue=0.5, parameterName=Contrasttest}
  ) {materialFunction=&HeightLerp}
  n53 = BreakMaterialAttributes!(Attr=StaticSwitchParameter(
    True=MaterialFunctionCall(Base Material=@n10.Blended Material, Top Material=@n51.Blended Material, Alpha=@n52.Alpha) {materialFunction=&MatLayerBlend_Standard}.Blended Material
    False=@n51.Blended Material
   ) {defaultValue=true, parameterName=Use_Cliffs})
  n54 = MakeMaterialAttributes!(
   BaseColor=@n53.BaseColor
   Metallic
   Specular=Multiply(A=@n53.BaseColor, B=ScalarParameter() {defaultValue=0.5, parameterName=Spec_Amount})
   Roughness=@n53.BaseColor
   Anisotropy
   EmissiveColor
   Opacity=@n53.BaseColor
   OpacityMask
   Normal=@n53.BaseColor
   Tangent
   WorldPositionOffset
   SubsurfaceColor
   ClearCoat
   ClearCoatRoughness
   AmbientOcclusion=@n53.BaseColor
   Refraction
   CustomizedUVs_0
   CustomizedUVs_1
   CustomizedUVs_2
   CustomizedUVs_3
   CustomizedUVs_4
   CustomizedUVs_5
   CustomizedUVs_6
   CustomizedUVs_7
   PixelDepthOffset
   ShadingModel
   Displacement
  )
  n55 = WorldPosition()
  n56 = ScalarParameter() {defaultValue=306200, group=TerrainColorMatch, parameterName=TerrainSize}
  n57 = Divide(A=@n56)
  n58 = Divide(A=Subtract(A=ComponentMask(None=@n55.XYZ) {g=true, r=true}, B=AppendVector(A=@n57, B=@n57)), B=@n56)
  n59 = TextureSample(UVs=@n58) {
   samplerSource=SSM_Wrap_WorldGroupSettings
   samplerType=SAMPLERTYPE_LinearColor
   texture=&T_Athena_Terrain_Biome_M
  }
  n60 = MaterialFunctionCall(
   Mask 2=TextureObject() {samplerType=SAMPLERTYPE_LinearColor, texture=&T_Athena_Terrain_Topo_Mask}
   UV Coordinates=@n58
   GrassMask=Constant() {r=1}
   Arid=@n59.B
   BaseTexture=MaterialFunctionCall(Material=@n54) {materialFunction=&MatLayerBlend_BreakBaseColor}.BaseColor
  ) {materialFunction=&MF_TerrainTopoAdjustment}
  n61 = Multiply(
   A=Desaturation(
    None=@n60.Result
    Fraction=CollectionParameter() {collection=&MPC_Landscape, parameterName=Arid_EdgeDesaturation}
   )
   B=ComponentMask(None=CollectionParameter() {collection=&MPC_Landscape, parameterName=Arid_SandSediment_Tint}) {b=true, g=true, r=true}
  )
  n62 = StaticSwitchParameter(
   True=MaterialFunctionCall(Mat_In=@n54, BaseColor=LinearInterpolate(A=@n60.Result, B=@n61, Alpha=@n60.Result)) {materialFunction=&MF_Athena_ReplaceBaseColor}.Result
   False=@n54
  ) {parameterName=UseTopo}
  n63 = MaterialFunctionCall!(Material=@n62) {materialFunction=/Engine/Functions/MaterialLayerFunctions/MatLayerBlend_BreakNormal}
  n64 = MaterialFunctionCall(Float3=@n63.Normal) {materialFunction=&BreakOutFloat3Components}
  n65 = Multiply(A=Transform(None=Constant3Vector() {constant=<rgba 0 0 1 0>}), B=Constant() {r=1})
  n66 = ComponentMask(None=@n65) {g=true}
  n67 = AppendVector(A=AppendVector(A=Multiply(A=@n64.R) {constB=-1}, B=Multiply(A=@n64.B, B=@n66)), B=Multiply(A=@n64.G) {constB=-1})
  n68 = ComponentMask(None=@n65) {r=true}
  n69 = AppendVector(A=AppendVector(A=Multiply(A=@n68, B=@n64.B), B=Multiply(A=@n64.R) {constB=-1}), B=Multiply(A=@n64.G) {constB=-1})
  n70 = Constant()
  n71 = MaterialFunctionCall(In=Abs(None=@n68), Contrast=@n70) {materialFunction=&CheapContrast}
  n72 = ComponentMask(None=Multiply(A=@n63.Normal, B=Constant3Vector() {constant=<rgba 1 1 1 0>})) {g=true, r=true}
  n73 = ComponentMask(None=@n65) {b=true}
  n74 = Multiply(A=@n64.B, B=@n73)
  n75 = Abs(None=@n73)
  n76 = Normalize(VectorInput=LinearInterpolate(
    A=LinearInterpolate(A=@n67, B=@n69, Alpha=@n71.Result)
    B=AppendVector(A=@n72, B=@n74)
    Alpha=MaterialFunctionCall(In=@n75, Contrast=@n70) {materialFunction=&CheapContrast}.Result
   ))
  n77 = MaterialFunctionCall!(
   Material=@n62
   Normal=MaterialFunctionCall(
    Normal=Transform(None=@n76) {transformSourceType=TRANSFORMSOURCE_World, transformType=TRANSFORM_Tangent}
    Flatness=ScalarParameter() {group=MasterControl, parameterName=NormalFlatten}
   ) {materialFunction=&FlattenNormal}.Result
  ) {materialFunction=/Engine/Functions/MaterialLayerFunctions/MatLayerBlend_ReplaceNormals}
  n78 = StaticSwitchParameter(True=@n77.Result, False=@n77.Result) {
   defaultValue=true
   desc=Weird legacy requirement for procedural grass to spawn.
   parameterName="Use Grass(true) or a BaseTilingTexture(False)"
  }
  n79 = SetMaterialAttributes!(
   MaterialAttributes=@n78
   World Position Offset=StaticSwitchParameter(
    True=MaterialFunctionCall() {materialFunction=/Game/Athena/Environments/Landscape/MaterialFunctions/MF_SplineMatHeightOffset}.Result
    False=@n78
   ) {parameterName=IsDesertRoad}
  ) {attributeSetTypes [#2]}
  n80 = BreakMaterialAttributes(Attr=@n79)
  n81 = MakeMaterialAttributes(
   BaseColor=@n80.BaseColor
   Specular=Multiply(A=@n80.BaseColor, B=ScalarParameter() {defaultValue=1, parameterName=SimpleMobileSpec})
   Roughness=Constant() {r=1}
   Normal=@n80.BaseColor
   WorldPositionOffset=@n80.BaseColor
  )
  n82 = QualitySwitch(Default=@n79, Low=@n81, High=@n79, Medium=@n79)
  MaterialFunctionCall!(In=MaterialFunctionCall!(In=ShadingPathSwitch(Default=@n82, Deferred=@n82, Mobile=@n81) {desc=Migrated from FeatureLevelSwitch}) {materialFunction=/BlastBerryMapContent/Game/TechArt/Material_Functions/MF_Rufus_RVTBlendingOn}.Result) {materialFunction=/Game/Athena/Asteria/Environments/Landscape/Materials/MaterialFunctions/MF_Terrain_LumenModification}
  n83 = Divide(
   A=ComponentMask(None=WorldPosition() {worldPositionShaderOffset=WPT_ExcludeAllShaderOffsets}.XYZ) {g=true, r=true}
   B=ScalarParameter() {defaultValue=25000, parameterName=MacroNormalScale}
  )
  n84 = ScalarParameter() {defaultValue=0.625, parameterName=MacroNormalIntensity}
  n85 = Constant() {r=0.9}
  n86 = ShadingPathSwitch(Default=Constant() {r=0.25}, Mobile=@n85) {desc=Migrated from FeatureLevelSwitch}
  n87 = Multiply(
   A=LinearInterpolate(
    A=Constant3Vector() {constant=<rgba 0 0 1 0>}
    B=TextureSample(UVs=@n83) {samplerType=SAMPLERTYPE_Normal, texture=/Game/ContentCreationTools/TextureCreation/Textures/T_BPCreated_MacroNormal_01}.RGB
    Alpha=MaterialFunctionCall() {materialFunction=/BlastBerryMapContent/Game/Athena/Environments/Landscape/Material/MF_TerrainDistanceFade}.Result
   )
   B=MaterialFunctionCall(X=@n84, Y=@n84, Z=QualitySwitch(Default=@n86, Low=@n85)) {materialFunction=&MakeFloat3}.Result
  )
  n88 = Add(
   A=Multiply(
    A=MaterialFunctionCall(Material=@n77.Result) {materialFunction=&MatLayerBlend_BreakNormal}.Normal
    B=Constant3Vector() {constant=<rgba 1 1 0.5 0>}
   )
   B=@n87
  )
  MaterialFunctionCall(
   Material=@n77.Result
   Normal=ShadingPathSwitch(Default=QualitySwitch(Default=@n88, Low=@n87, High=@n88, Medium=@n88), Deferred=@n88, Mobile=@n87) {desc=Migrated from FeatureLevelSwitch}
  ) {materialFunction=&MatLayerBlend_ReplaceNormals}
  Multiply(A=@n80.BaseColor, B=VectorParameter() {parameterName=GlobalTint, defaultValue=<rgba 0.9323 0.8938 0.7258 1>}.RGB)
  Multiply(A=@n38, B=ScalarParameter())
  n89 = VectorParameter() {parameterName=mud tint, defaultValue=<rgba 3.2 2.9781 2.8672 1.5>}
  Multiply(A=Multiply(A=@n28, B=@n89.RGB), B=@n89.RGB)
  n90 = QualitySwitch(Default=@n54, Low=@n54, High=@n62, Medium=@n62)
  ShadingPathSwitch(Default=@n90, Deferred=@n90, Mobile=@n54) {desc=Migrated from FeatureLevelSwitch}
  StaticBool()
  StaticSwitchParameter() {parameterName=IsDesertRoad_1}
  VectorParameter() {parameterName=Arid_Gravel_Tint, defaultValue=<rgba 1 1 1 0>}
  VertexInterpolator(VS=@n58)
  VertexNormalWS()
 ]
}
